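# Unlit translucent Niagara sprite material for auras that paints a decal-like Fresnel glow near scene geometry.
# A Fresnel on flattened scene world normals, tuned by dynamic parameters, is masked by a UV sphere mask and inverted depth fade to set opacity with particle alpha.
# Squared particle color drives emissive through an FX adjustment function.
# Mobile and low quality replace the Fresnel with a Forward vector parameter.
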Material /SparksCosmetics/Items/Auras/Materials/M_Aura_Decal_Fresnel
graph {
 properties {bUsedWithNiagaraSprites=true, blendMode=BLEND_Translucent, pixelDepthOffsetMode=PDOM_Legacy, shadingModel=MSM_Unlit, usedShadingModels=Unlit}
 nodes [
  n1 = ParticleColor()
  MaterialFunctionCall(Emissive=Multiply(A=ScalarParameter() {defaultValue=5, parameterName=EmissiveMult}, B=Multiply(A=@n1.RGB, B=@n1.RGB))) {materialFunction=&MF_FX_EmissiveAdjustments}
  n2 = DynamicParameter() {defaultValue=<rgba 3 0.15 1 1>, paramNames [FresnelExponent, FlattenNormal, Param3, Param4]}
  n3 = SceneTexture() {sceneTextureId=PPI_WorldNormal}
  n4 = Fresnel(
   ExponentIn=@n2.FresnelExponent
   Normal=MaterialFunctionCall(Normal=ComponentMask(None=@n3.Color) {b=true, g=true, r=true}, Flatness=@n2.FlattenNormal) {materialFunction=&FlattenNormal}.Result
  ) {baseReflectFraction=0, exponent=1}
  n5 = VectorParameter() {parameterName=Forward, defaultValue=<rgba 0 0 1 1>}
  n6 = ShadingPathSwitch(Default=QualitySwitch(Default=@n4, Low=@n5.RGB, High=@n4, Medium=@n4, Epic=@n4), Mobile=@n5.RGB)
  n7 = SphereMask(
   A=TextureCoordinate()
   B=Constant() {r=0.5}
   Radius=ScalarParameter() {defaultValue=0.5, parameterName=SphereMaskRadius}
   Hardness=ScalarParameter() {parameterName=SphereMaskHardness}
  ) {hardnessPercent=0}
  n8 = OneMinus(None=DepthFade(FadeDistance=ScalarParameter() {defaultValue=15, parameterName=VisibleDepthDistance}) {fadeDistanceDefault=80})
  n9 = ScalarParameter() {defaultValue=15, parameterName=DepthFade}
  Multiply(A=@n1.A, B=Multiply(A=Multiply(A=@n6, B=@n7), B=DepthFade(Opacity=@n8, FadeDistance=@n9) {fadeDistanceDefault=15}))
 ]
}
