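# Additive unlit Niagara sprite material for an aura flare: a halo corona with rotating, flickering rays.
# Two rotating spikeball samples, one inverted and sharpened, form a contrast-boosted ray mask alongside a sphere-mask-remapped corona gradient texture.
# Dynamic parameters drive size, rotation, multi-sine flicker and color jitter, with static switches for particle color and hue shift.
# Depth and near-camera fades apply only on the default shading path; the packed index-1 parameters repeat ErodeSize in every channel.
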Material /SparksCosmetics/Items/Auras/Materials/M_Aura_Flare_Halo_Corona
graph {
 properties {
  bUsedWithNanite=true
  bUsedWithNiagaraSprites=true
  blendMode=BLEND_Additive
  pixelDepthOffsetMode=PDOM_Legacy
  shadingModel=MSM_Unlit
  twoSided=true
  usedShadingModels=Unlit
 }
 nodes [
  MaterialFunctionCall(Emissive=NamedRerouteUsage()) {materialFunction=&MF_FX_EmissiveAdjustments}
  n1 = DynamicParameter() {defaultValue=<rgba 0.6 1 1 1>, paramNames [Size,RotateSpeed,Flicker,ColorJitter]}
  NamedRerouteDeclaration(None=@n1.ColorJitter) {name=ColorJitter On/Off, nodeColor=<rgba 1 0 0.9844 1>}
  NamedRerouteDeclaration(None=@n1.Flicker) {name=Flicker On/Off, nodeColor=<rgba 0.1615 0.075 0.005 1>}
  n2 = Multiply(A=ScalarParameter() {defaultValue=1, parameterName=Brightness}) {constB=6}
  n3 = Add(A=TextureCoordinate() {uTiling=0.25, vTiling=0.25}) {constB=0.375}
  n4 = NamedRerouteUsage()
  n5 = MaterialFunctionCall(Float4=LinearInterpolate(
    A=Constant4Vector() {constant=<rgba 6.667 0 1 0>}.RGBA
    B=NamedRerouteUsage() {desc="R= ErosionSize, G= Erosion Amount, B= DepthFadeMult, "}
    Alpha=ScalarParameter() {parameterName=UseIndex1Params}
   )) {materialFunction=&BreakOutFloat4Components}
  n6 = TextureSample(UVs=Rotator(
    Coordinate=MaterialFunctionCall(Texture Scale=@n5.R) {materialFunction=&ScaleUVsByCenter}.UVs
    Time=NamedRerouteUsage()
   ) {speed=-0.72}) {texture=/Game/Effects/Fort_Effects/Textures/Flares/T_Flare_Spikeball}
  n7 = Multiply(
   A=TextureSample(UVs=Rotator(Coordinate=@n3, Time=@n4)) {texture=&T_Flare_Spikeball}.RGB
   B=Power(Base=OneMinus(None=@n6.RGB)) {constExponent=5}
  )
  NamedRerouteDeclaration(None=MaterialFunctionCall(In=Multiply(A=@n2, B=@n7) {constB=80, desc=Mask Boost}, Contrast=NamedRerouteUsage()) {materialFunction=&CheapContrast}.Result) {name=Ray Mask, nodeColor=<rgba 0 1 0.25 1>}
  n8 = DynamicParameter() {desc=Niagara index 1, parameterIndex=1, defaultValue=<rgba 6.667 0 1 1>, paramNames [ErodeSize, Erode, DepthFadeMult, Param4]}
  NamedRerouteDeclaration(None=MaterialFunctionCall(R=@n8.ErodeSize, G=@n8.ErodeSize, B=@n8.ErodeSize, A=@n8.ErodeSize) {materialFunction=&AppendMany}.RGBA) {name=Packed Dynamic Parameter Index1, nodeColor=<rgba 0.3594 1 0 1>}
  n9 = DepthFade(
   Opacity=MaterialFunctionCall(Distance=ScalarParameter() {defaultValue=200, parameterName=NearCamFade}) {materialFunction=&Near_Camera_Fade}.Result
   FadeDistance=Multiply(A=ScalarParameter() {defaultValue=100, parameterName=DepthFade}, B=NamedRerouteUsage())
  )
  n10 = Constant() {r=1}
  NamedRerouteDeclaration(None=Multiply(A=NamedRerouteUsage(), B=ShadingPathSwitch(Default=@n9, Forward=@n10, Mobile=@n10))) {name=Opacity, nodeColor=<rgba 1 0.2578 0 1>}
  n11 = NamedRerouteUsage()
  n12 = Multiply(A=Multiply(A=NamedRerouteUsage(), B=NamedRerouteUsage()), B=Multiply(A=NamedRerouteUsage(), B=NamedRerouteUsage()) {desc=Mask Combine}) {desc=Apply flicker}
  n13 = ShadingPathSwitch(
   Default=MaterialFunctionCall(LDR Input=Saturate(None=@n12), HDR Tint=@n12) {materialFunction=&DeriveHDRfromLDR}.RemappedRange
   Forward=@n12
   Mobile=@n12
  )
  NamedRerouteDeclaration(None=StaticSwitchParameter(
    True=MaterialFunctionCall(Hue Shift Percentage=@n11, Texture=@n13) {materialFunction=&HueShift}.Result
    False=@n13
   ) {parameterName=UseColorJitter?}) {name=Emissive Color, nodeColor=<rgba 1 0 0.5156 1>}
  NamedRerouteDeclaration(None=@n1.Size) {name=Size, nodeColor=<rgba 0.1484 1 0 1>}
  NamedRerouteDeclaration(None=Multiply(A=Time(), B=@n1.RotateSpeed)) {name=RotateSpeed, nodeColor=<rgba 1 0.0469 0.0719 1>}
  n14 = ParticleColor()
  NamedRerouteDeclaration(None=@n14.RGB) {name=Particle RGB, nodeColor=<rgba 1 0.5391 0 1>}
  NamedRerouteDeclaration(None=@n14.A) {name=Particle A, nodeColor=<rgba 0 0.4844 1 1>}
  NamedRerouteDeclaration(None=@n5.G) {name=Erode, nodeColor=<rgba 1 0.75 0 1>}
  NamedRerouteDeclaration(None=@n5.B) {name=Depth Fade, nodeColor=<rgba 0 1 0.7188 1>}
  n15 = ParticleRandom()
  n16 = Multiply(A=Time(), B=Add(A=Multiply(A=@n15) {constB=0.1}))
  n17 = Sine(None=@n16) {period=0.155}
  n18 = Sine(None=@n16) {period=0.2}
  n19 = Multiply(A=Sine(None=@n16) {period=0.075}, B=Multiply(A=Multiply(A=@n17, B=@n18), B=Sine(None=@n16) {period=0.43}))
  NamedRerouteDeclaration(None=LinearInterpolate(A=Clamp(None=@n19) {maxDefault=3}, Alpha=NamedRerouteUsage()) {constA=1}) {name=Flicker Mask, nodeColor=<rgba 0 0.3437 1 1>}
  n20 = NamedRerouteUsage()
  n21 = Constant() {r=0.5}
  n22 = TextureSampleParameter2D(UVs=SphereMask(A=TextureCoordinate(), B=@n21, Radius=@n21, Hardness=LinearInterpolate(Alpha=@n20) {constA=1, constB=0}) {hardnessPercent=25}) {parameterName=Halo Texture, texture=/Game/Effects/Fort_Effects/Textures/Grad/T_Halo_Corona_Gradient}
  NamedRerouteDeclaration(None=StaticSwitchParameter(True=Desaturation(None=@n22.RGB), False=@n22.RGB) {defaultValue=true, parameterName=UseParticleColor}) {name=Corona Gradient Mask or Color, nodeColor=<rgba 0 1 0.0391 1>}
  NamedRerouteUsage()
 ]
}
