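# Translucent unlit material for aura Niagara mesh particles that draws them as soft, two-color radial sprites.
# World position offset combines the Sprite function with a camera offset, both driven by dynamic parameters.
# A UV sphere mask, using radius as hardness too, blends edge to center color, tinted by particle color through emissive adjustments.
# Opacity multiplies particle alpha, that mask and an inverted world-space mask around the offset particle position.
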Material /SparksCosmetics/Items/Auras/Materials/M_Aura_MeshParticle_Sprite
graph {
 properties {
  bUsedWithNiagaraMeshParticles=true
  blendMode=BLEND_Translucent
  pixelDepthOffsetMode=PDOM_Legacy
  shadingModel=MSM_Unlit
  usedShadingModels=Unlit
 }
 nodes [
  n1 = DynamicParameter() {defaultValue=<rgba 10 5 0.5 0.5>, paramNames [SpriteSize, Camera Offset, Sphere Mask Radius, Sphere Mask Hardness]}
  Add(
   A=MaterialFunctionCall!(Center Location=WorldPosition().XYZ, 0-1 UVs, XY scale=AppendVector(A=@n1.SpriteSize, B=@n1.SpriteSize), OverridePivot, "Normals (Optional)") {materialFunction=&Sprite}.World position offset
   B=MaterialFunctionCall(Offset Amount=@n1.Camera Offset, World Position=WorldPosition().XYZ) {materialFunction=/Engine/Functions/Engine_MaterialFunctions02/WorldPositionOffset/CameraOffset}.Clamped Camera Offset
  )
  n2 = VectorParameter() {parameterName=Edge Color, defaultValue=<rgba 0.6896 0.6771 1 1>}
  n3 = VectorParameter() {parameterName=Center Color, defaultValue=<rgba 1 0.9335 0.5823 1>}
  n4 = DynamicParameter() {defaultValue=<rgba 10 5 0.5 0.5>, paramNames [SpriteSize, Camera Offset, Sphere Mask Radius, Sphere Mask Hardness]}
  n5 = SphereMask(A=TextureCoordinate(), B=Constant() {r=0.5}, Radius=@n4.Sphere Mask Radius, Hardness=@n4.Sphere Mask Radius)
  n6 = ParticleColor()
  MaterialFunctionCall(Emissive=Multiply(A=@n6.RGB, B=LinearInterpolate(A=@n2.RGB, B=@n3.RGB, Alpha=@n5))) {materialFunction=&MF_FX_EmissiveAdjustments}
  n7 = WorldPosition()
  n8 = Add(A=ParticlePositionWS(), B=VectorParameter() {parameterName=ParticlePositionOffset, defaultValue=<rgba 0 0 300 0>}.RGB)
  Multiply(
   A=Multiply(A=@n6.A, B=@n5)
   B=OneMinus(None=SphereMask(A=@n7.XYZ, B=@n8, Radius=@n4.Sphere Mask Radius, Hardness=@n4.Sphere Mask Hardness) {attenuationRadius=500})
  )
 ]
}
